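; Player pawn component for the AmberVine Force power, tracking its channel, target hit result and damage.
; It broadcasts channel start/stop, replicated hit-result updates and defeated targets, and exposes remaining channel time.
; Weapon-equip events add or remove the AmberVine sound library on Galileo Lobster weapons.
; The authority accumulates activations, channel durations, non-friendly damage and eliminations, submitting them as one analytics event on EndPlay.
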(event EventBeginPlay
  (|CallParentFunction)
  (CallFunction|ResetAnalyticsVariables)
  (bind _asfort_player_pawn (Utilities|Casting|CastToFortPlayerPawn (Components|GetOwner)))
  (bind _output_get (Variables|Default|SetPlayerPawn _asfort_player_pawn))
  (EventDispatchers|BindEventtoOnDidDamage _output_get (EventDispatchers|CreateEvent :!event OnDamagedOtherActor))
  (FortPawn|BindEventtoOnWeaponEquippedDelegate _output_get (EventDispatchers|CreateEvent :!event Custom|OnEquipWeapon))
  (CallFunction|OnEquipWeapon (Variables|/Script/FortniteGame.FortPawn|GetCurrentWeapon (Variables|Default|GetPlayerPawn))))

(event EventEndPlay
  (params (EndPlayReason :type EEndPlayReason%20Enum :default "Destroyed"))
  (Utilities|FlowControl|Sequence
    (:then_0
      (bind _playerpawn (Variables|Default|GetPlayerPawn)
        (:then
          (EventDispatchers|UnbindEventfromOnDidDamage _playerpawn (EventDispatchers|CreateEvent :!event OnDamagedOtherActor))
          (FortPawn|UnbindEventfromOnWeaponEquippedDelegate (Variables|Default|GetPlayerPawn) (EventDispatchers|CreateEvent :!event Custom|OnEquipWeapon))
          (bind _asb_athena_galileo_lobster_parent (Utilities|Casting|CastToB_Athena_Galileo_Lobster_Parent (Variables|/Script/FortniteGame.FortPawn|GetCurrentWeapon (Variables|Default|GetPlayerPawn))))
          (SoundLibrary|RemoveLibrary (Variables|/Script/FortniteGame.FortWeapon|GetSoundLibraryComponent _asb_athena_galileo_lobster_parent) (Variables|Default|GetAmberVine%20Sound%20Library) :!class /Script/SoundLibrary.SoundLibraryComponent))
        (:else)))
    (:then_1
      (Analytics|SubmitAnalytics))))

(event Custom|OnTargetDefeated
  (Networking|CanExecuteCosmeticEvents
    (:True
      (Default|CallOnDefeatedTarget))))

(event Custom|OnEquipWeapon
  (params (NewWeapon :type Fort%20Weapon%20Object%20Reference) (PrevWeapon :type Fort%20Weapon%20Object%20Reference))
  (bind (_asb_athena_galileo_lobster_parent _bsuccess) (Utilities|Casting|CastToB_Athena_Galileo_Lobster_Parent NewWeapon))
  (bind (_asb_athena_galileo_lobster_parent_2 _bsuccess_2) (Utilities|Casting|CastToB_Athena_Galileo_Lobster_Parent PrevWeapon))
  (Utilities|FlowControl|Sequence
    (:then_0
      (if _bsuccess
        (SoundLibrary|AddLibrary (Variables|/Script/FortniteGame.FortWeapon|GetSoundLibraryComponent _asb_athena_galileo_lobster_parent) (Variables|Default|GetAmberVine%20Sound%20Library) :!class /Script/SoundLibrary.SoundLibraryComponent)))
    (:then_1
      (if _bsuccess_2
        (SoundLibrary|RemoveLibrary (Variables|/Script/FortniteGame.FortWeapon|GetSoundLibraryComponent _asb_athena_galileo_lobster_parent_2) (Variables|Default|GetAmberVine%20Sound%20Library) :!class /Script/SoundLibrary.SoundLibraryComponent)))))

(fn AddToChannelDurationAnalytics
  (bind _row_maxchanneltime (Variables|Default|GetRow_MaxChannelTime))
  (bind _returnvalue (ScalableFloat|GetValueatLevel _row_maxchanneltime))
  ; Analytics for Channel Duration
  (Utilities|SwitchHasAuthority
    (:Authority
      (if (not (Replay|GetIsPlayingReplay))
        (bind _output_get (LocalVariables|SetChannelDuration (- (Utilities|Time|GetGameTimeinSeconds) (Variables|Default|GetChannelStartTime))))
        (if (Math|Float|InRange(Float) _output_get 0.5 (+ _returnvalue 2.0))
          (Utilities|Array|Add (Variables|Analytics|GetChannelTimes) (Math|Float|Clamp(Float) (LocalVariables|GetChannelDuration) 0.0 _returnvalue)))))))

(fn AddToTotalDamageAnalytics
  (params (Damage :type Float%20%28double-precision%29))
  (Utilities|SwitchHasAuthority
    (:Authority
      (if (not (Replay|GetIsPlayingReplay))
        (Variables|Analytics|SetTotalPawnDamage (+ (Variables|Analytics|GetTotalPawnDamage) (Variables|GetDamage)))))))

(fn GetAverageChannelDuration
  (bind _channeltimes (Variables|Analytics|GetChannelTimes))
  (LocalVariables|SetNumOfChannels (Utilities|Array|Length _channeltimes))
  (for _array_element _channeltimes
    (LocalVariables|SetTotalOfAllDurations (+ _array_element (LocalVariables|GetTotalOfAllDurations))))
  (return (Math|Float|SafeDivide (LocalVariables|GetTotalOfAllDurations) (Math|Conversions|ToFloat(Integer) (LocalVariables|GetNumOfChannels)))))

(fn GetChannelTimeRemaining
  (return (> (Variables|Default|GetChannelStartTime) 0) (select (> (Variables|Default|GetChannelStartTime) 0) (Math|Float|MapRangeClamped (Math|Float|SafeDivide (- (Utilities|Time|GetGameTimeinSeconds) (Variables|Default|GetChannelStartTime)) (ScalableFloat|GetValueatLevel (Variables|Default|GetRow_MaxChannelTime))) 0.0 1.0 1.0) -1.0)))

(fn GetCurrentHitResult
  (return (Variables|Default|GetCurrentHitResult)))

(fn GetCurrentLocation
  (return (Variables|Default|GetCurrentLocation)))

(fn IncrementActivationAnalytics
  (Utilities|SwitchHasAuthority
    (:Authority
      (if (not (Replay|GetIsPlayingReplay))
        (Math|Integer|IncrementInt (Variables|Analytics|GetActivations))))))

(fn IncrementEliminationAnalytics
  (Utilities|SwitchHasAuthority
    (:Authority
      (if (not (Replay|GetIsPlayingReplay))
        (Math|Integer|IncrementInt (Variables|Analytics|GetEliminations))))))

(fn OnDamagedOtherActor
  (params (DamagedActor :type Actor%20Object%20Reference) (Damage :type Float%20%28single-precision%29) (InstigatedBy :type Controller%20Object%20Reference) (DamageCauser :type Actor%20Object%20Reference) (HitLocation :type Vector) (FHitComponent :type Primitive%20Component%20Object%20Reference) (BoneName :type Name) (Momentum :type Vector))
  (Networking|SwitchHasAuthority
    (:Authority
      (bind _asfort_pawn (Utilities|Casting|CastToFortPawn DamagedActor))
      (bind _output_get (LocalVariables|SetDamagedPlayerPawn _asfort_pawn))
      (if (Utilities|Enum|NotEqual(Enum) (Team|GetTeamAffiliation _output_get (Variables|Default|GetPlayerPawn)) "Friendly")
        (Analytics|AddtoTotalDamageAnalytics (Variables|GetDamage))
        (if (and (> (Variables|GetDamage) 0) (FortPawn|IsDead (LocalVariables|GetDamagedPlayerPawn)))
          (CallFunction|OnTargetDefeated)
          (Analytics|IncrementEliminationAnalytics))))))

(fn OnRep_CurrentHitResult
  (CallFunction|ParseCurrentHitResultData)
  (Default|CallOnCurrentHitResultUpdated))

(fn ParseCurrentHitResultData
  (bind _currenthitresult (Variables|Default|GetCurrentHitResult))
  (bind (_bblockinghit _binitialoverlap _time _distance _location _impactpoint _normal _impactnormal _physmat _hitactor _hitcomponent _hitbonename _bonename _hititem _elementindex _faceindex _tracestart _traceend) (Collision|BreakHitResult _currenthitresult))
  (Utilities|FlowControl|Sequence
    (:then_0
      (Variables|Default|SetValidHitResult _bblockinghit)
      (Variables|Default|SetCurrentLocation (select (Utilities|Operators|NotEqual(!=) :A _location) _location _traceend)))))

(fn ResetAnalyticsVariables
  ; Clear analytics variables
  (Utilities|Array|Clear (Variables|Analytics|GetChannelTimes))
  (Variables|Analytics|SetActivations 0)
  (Variables|Analytics|SetTotalPawnDamage 0.0)
  (Variables|Analytics|SetEliminations 0))

(fn SetCurrentHitResult
  (params (CurrentHitResult :type Hit%20Result%20Structure))
  (Variables|Default|SetCurrentHitResult CurrentHitResult)
  (Networking|CanExecuteCosmeticEvents
    (:True
      (CallFunction|ParseCurrentHitResultData)
      (Default|CallOnCurrentHitResultUpdated))))

(fn StartChannelTime
  (Variables|Default|SetChannelStartTime (Utilities|Time|GetGameTimeinSeconds))
  (Default|CallOnChannelStart)
  (Utilities|SwitchHasAuthority
    (:Authority
      (if (not (Replay|GetIsPlayingReplay))
        (Math|Integer|IncrementInt (Variables|Analytics|GetActivations))))))

(fn StopChannelTime
  (Utilities|FlowControl|Sequence
    (:then_0
      (Analytics|AddtoChannelDurationAnalytics))
    (:then_1
      (Variables|Default|SetChannelStartTime 0.0)
      (Default|CallOnChannelStop))))

(fn SubmitAnalytics
  (Utilities|SwitchHasAuthority
    (:Authority
      (Utilities|FlowControl|Sequence
        (:then_0
          ; Add activations
          (bind _returnvalue (Analytics|ConstructAnalyticsAttribute :AttributeName "Activations" :IntValue (Variables|Analytics|GetActivations) :floatValue -1.0 :BoolValue false))
          (Utilities|Array|Add (LocalVariables|GetLocalAttributes) _returnvalue)
          ; Add average channel duration
          (bind _returnvalue_2 (Analytics|ConstructAnalyticsAttribute :AttributeName "AverageChannelDuration" :IntValue -1 :floatValue (Analytics|GetAverageChannelDuration) :BoolValue false))
          (Utilities|Array|Add (LocalVariables|GetLocalAttributes) _returnvalue_2)
          ; Add total pawn damage
          (bind _returnvalue_3 (Analytics|ConstructAnalyticsAttribute :AttributeName "PawnDamage" :IntValue -1 :floatValue (Variables|Analytics|GetTotalPawnDamage) :BoolValue false))
          (Utilities|Array|Add (LocalVariables|GetLocalAttributes) _returnvalue_3)
          ; Add total pawn damage
          (bind _returnvalue_4 (Analytics|ConstructAnalyticsAttribute :AttributeName "PawnEliminations" :IntValue (Variables|Analytics|GetEliminations) :floatValue -1.0 :BoolValue false))
          (Utilities|Array|Add (LocalVariables|GetLocalAttributes) _returnvalue_4))
        (:then_1
          (Analytics|FireAnalyticsEventBlueprintArray "Core.PlotTwist.AmberVine" (LocalVariables|GetLocalAttributes)))))))
