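# Unlit, two-sided translucent material for the Ambervine lightning bolt Niagara mesh particles.
# World position offset combines SplineThicken width, scaled by camera distance, with a masked sine wave and vertical arc, driven by dynamic parameters.
# A Travel-driven sphere mask, particle alpha, panning noise and optional edge fade build the bolt's masks.
# Particle color feeds a shared emissive adjustment function.
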Material /PlotTwist/ForcePowers/AmberVine/FX/M_Ambervine_LightningBolt
graph {
 properties {
  bUsedWithNiagaraMeshParticles=true
  blendMode=BLEND_Translucent
  pixelDepthOffsetMode=PDOM_Legacy
  shadingModel=MSM_Unlit
  twoSided=true
  usedShadingModels=Unlit
 }
 nodes [
  n1 = Sine(None=ComponentMask(None=TextureCoordinate()) {g=true}) {period=2}
  n2 = Multiply(A=VertexColor().B, B=DynamicParameter() {defaultValue=<rgba 1 0 1 0>, paramNames [Width,WidthBase,WidthTip,VerticalArc]}.Width)
  n3 = DynamicParameter() {defaultValue=<rgba 1 0 1 0>, paramNames [Width,WidthBase,WidthTip,VerticalArc]}
  n4 = Multiply(A=@n2, B=Multiply(A=ScalarParameter() {defaultValue=1, parameterName=WidthBase_Mult}, B=@n3.WidthBase))
  n5 = ScalarParameter() {defaultValue=1, parameterName=WidthTip_Mult}
  n6 = Multiply(A=Multiply(A=@n2, B=Multiply(A=@n3.WidthTip, B=@n5)), B=OneMinus(None=VertexColor().R))
  n7 = LinearInterpolate(
   B=ScalarParameter() {defaultValue=1, parameterName=DistanceWidthMult}
   Alpha=MaterialFunctionCall(Distance=ScalarParameter() {defaultValue=25000, parameterName=DistanceWidthMultRange}) {materialFunction=&Near_Camera_Fade}.Result
  ) {constA=1}
  n8 = VectorParameter() {parameterName=WPOWaveDirection, defaultValue=<rgba 0 1 0.5 0>}
  n9 = TextureCoordinate()
  n10 = DynamicParameter() {parameterIndex=1, defaultValue=<rgba 1 0 1 0>, paramNames [WPOWaveTime, WPOWaveMult, WPOWaveLength, VerticalArc]}
  n11 = Add(A=Divide(A=ComponentMask(None=@n9) {g=true}, B=@n10.WPOWaveLength), B=@n10.WPOWaveTime)
  n12 = MaterialFunctionCall(Y=Sine(None=@n11), Z=Sine(None=Add(A=@n11) {constB=-0.75}) {period=0.75}) {materialFunction=&MakeFloat3}
  n13 = Multiply(A=Transform(None=Multiply(A=@n8.RGB, B=@n12.Result)) {transformSourceType=TRANSFORMSOURCE_Local}, B=@n10.WPOWaveMult)
  n14 = TextureCoordinate()
  n15 = Constant() {r=0.5}
  n16 = Power(Base=SphereMask(A=ComponentMask(None=@n14) {g=true}, B=@n15, Radius=@n15) {hardnessPercent=0}) {constExponent=0.5}
  n17 = TextureCoordinate()
  n18 = SphereMask(
   A=Power(Base=ComponentMask(None=@n17) {g=true}, Exp=ScalarParameter() {defaultValue=0.5, parameterName=VerticalArcPower})
   B=@n15
   Radius=@n15
  ) {hardnessPercent=0}
  n19 = Multiply(A=@n10.VerticalArc, B=Constant3Vector() {constant=<rgba 0 0 1 0>})
  Add(
   A=MaterialFunctionCall(WidthBase=Multiply(A=@n1, B=@n4), WidthTip=Multiply(A=@n6, B=@n7)) {materialFunction=&SplineThicken}.WorldPositionOffset
   B=Add(A=Multiply(A=@n13, B=@n16), B=Multiply(A=@n18, B=@n19))
  )
  n20 = Sine(None=TextureCoordinate()) {period=2}
  n21 = Multiply(
   A=Power(Base=ComponentMask(None=@n20) {r=true}, Exp=ScalarParameter() {defaultValue=2.2, parameterName=EdgeFade_Power})
   B=ScalarParameter() {defaultValue=1, parameterName=EdgeFade_Mult}
  )
  n22 = ParticleColor()
  n23 = OneMinus(None=@n22.A)
  n24 = ComponentMask(None=TextureCoordinate()) {g=true}
  n25 = Subtract(A=DynamicParameter() {parameterIndex=2, paramNames [Travel,WidthMult,WobbleMult,Random]}.Travel, B=VertexColor().R)
  n26 = LinearInterpolate(Alpha=@n25) {constA=-1, constB=2}
  n27 = Clamp(None=Add(A=@n23, B=SphereMask(A=@n24, B=@n26, Hardness=@n25) {attenuationRadius=1, hardnessPercent=50}))
  LinearInterpolate(A=@n27, B=Multiply(A=@n27, B=Clamp(None=@n21)), Alpha=ScalarParameter() {parameterName=Use_EdgeFade})
  n28 = ComponentMask(None=TextureCoordinate()) {g=true}
  Multiply(A=@n28, B=OneMinus(None=@n28))
  n29 = TextureSample(UVs=Panner(Coordinate=TextureCoordinate() {uTiling=0, vTiling=2}, Time=@n22.A) {speedY=-4}) {samplerType=SAMPLERTYPE_LinearColor, texture=&HighContrastNoise}
  Multiply(A=Multiply(A=Multiply(A=@n29.R) {constB=-1}, B=@n23), B=@n27)
  Reroute(None=MaterialFunctionCall(Emissive=ParticleColor().RGB) {materialFunction=&MF_FX_EmissiveAdjustments}.Result)
 ]
}
