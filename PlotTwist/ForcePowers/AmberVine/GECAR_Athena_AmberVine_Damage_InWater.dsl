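; GameplayEffect application requirement for AmberVine water damage that decides whether the target ability system's owner counts as in water.
; Vehicles report their own in-water state through the FortVehicleInterface.
; Player pawns with a current water body qualify, but those in vehicles also need the open-seat tag.
; The fallback return passes only owners that cast to BGA_Athena_FlopperSpawn_World.
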(fn CanApplyGameplayEffect
  (params (GameplayEffect :type Gameplay%20Effect%20Object%20Reference) (Spec :type Gameplay%20Effect%20Spec%20Structure%20%28by%20ref%29) (ASC :type Ability%20System%20Component%20Object%20Reference))
  (bind _returnvalue (Components|GetOwner ASC))
  (bind (_asbga_athena_flopper_spawn_world _bsuccess) (Utilities|Casting|CastToBGA_Athena_FlopperSpawn_World _returnvalue))
  (Utilities|FlowControl|Sequence
    (:then_0
      (bind _asfort_vehicle_interface (Utilities|Casting|CastToFortVehicleInterface _returnvalue))
      (return (Class|FortVehicleInterface|IsinWater _asfort_vehicle_interface)))
    (:then_1
      (bind _asfort_player_state (Utilities|Casting|CastToFortPlayerState _returnvalue))
      (Utilities|IsValid (PlayerState|GetCurrentPawn _asfort_player_state)
        (:Is%20Valid
          (if (Utilities|IsValid (WaterBody|GetCurrentWaterBody (PlayerState|GetCurrentPawn _asfort_player_state)) :!class /Script/Engine.KismetSystemLibrary)
            (if (Vehicle|IsinVehicle (PlayerState|GetCurrentPawn _asfort_player_state))
              (return (GameplayTags|HasMatchingGameplayTag (PlayerState|GetCurrentPawn _asfort_player_state) (Variables|Default|GetInOpenSeatOfVehicleTag)))
              (else
                (return true)))))))
    (:then_2
      ;[1 0.5 0] default return
      (return _bsuccess))))
