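; Custom application requirement that decides whether the AmberVine disable (EMP) gameplay effect can apply to a target.
; It rejects owners that cast to B_BGA_Athena_SilverBlazer_V2, the shield bubble, so they are not destroyed.
; Any other target qualifies if its ability system component matches the EMP-can-be-applied tag, or if its owner is a Fort vehicle and the scalable-float EMP vehicle toggle is on.
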(fn CanApplyGameplayEffect
  (params (GameplayEffect :type Gameplay%20Effect%20Object%20Reference) (Spec :type Gameplay%20Effect%20Spec%20Structure%20%28by%20ref%29) (ASC :type Ability%20System%20Component%20Object%20Reference))
  (bind (_asfort_vehicle_interface _bsuccess) (Utilities|Casting|CastToFortVehicleInterface (LocalVariables|GetOwner)))
  (bind _output_get (LocalVariables|SetOwner (Components|GetOwner ASC)))
  ; Don't destroy shield bubbles. They die from emp
  (Utilities|Casting|CastToB_BGA_Athena_SilverBlazer_V2 _output_get
    (:then
      (return false))
    (:CastFailed
      (return (or (GameplayTags|HasMatchingGameplayTag ASC (Variables|Default|GetTag_EMPCanBeApplied)) (and _bsuccess (ScalableFloat|GetValueasBool (Variables|Default|GetRow_EmpVehicleEnabled))))))))
